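# List entry widget that previews one item definition, apparently for a tech-art weapon mod tool, as a single text label.
# When given a list object, it casts it to FortItemDefinition and shows the object name, or "invalid item:" plus the object's display name if the cast fails.
# Selection changes switch the label color between white and yellow.
# An item display-name lookup is present but unused.
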WidgetBlueprint /Game/ArtTools/Tech_Art/WeaponMods/EW_ModToolItemDefPreviewListEntry
graphs [EventGraph {nodes [
   n1 = Event On Item Selection Changed!() -> (OutputDelegate, then, bIsSelected:bool) {is=title K2Node_Event}
   > SetColorAndOpacity(
    self=Get TextBlock() -> (TextBlock:Text Object Reference)
    InColorAndOpacity=Select(
     Option 0:Slate Color Structure=Make Slate Color(SpecifiedColor=(R=1,G=1,B=1,A=1))
     Option 1:Slate Color Structure=Make Slate Color(SpecifiedColor=(R=1,G=1,B=0,A=1))
     Index=@n1.bIsSelected
    ) -> (ReturnValue:Slate Color Structure)
   )
   n2 = Event OnListItemObjectSet()
   > n3 = Cast To FortItemDefinition!(execute, Object=@n2.ListItemObject) -> (then, CastFailed, AsFort Item Definition:Fort Item Definition Object Reference) {is=title K2Node_DynamicCast}
   > Set Text(Text="To Text (String)"(InString=GetObjectName(Object=@n3)), self=Get TextBlock())
   n4 = GetDisplayName(Object=@n2.ListItemObject)
   Set Text(execute=@n3.CastFailed, Text="To Text (String)"(InString=Append(A="invalid item: ", B=@n4)), self=Get TextBlock())
   GetDisplayName!(self:Item Definition Base Object Reference=@n3) -> (ReturnValue:Text)
  ]}]
variables []
widgets {
 info {inheritedWidgetCount=0, parentClass=&UserWidget, rootWidgetClass=&TextBlock, widgetCount=1}
 widgets [TextBlock {
   widgetClassPath=&TextBlock
   bIsVariable=true
   widget=TextBlock
   properties {
    text=ItemDef Name
    font {
     bForceMonospaced=false
     bMaterialIsStencil=false
     fontMaterial=null
     fontObject=&Roboto
     letterSpacing=0
     monospacedWidth=1
     size=12
     skewAmount=0
     typefaceFontName=Bold
     outlineSettings {
      bApplyOutlineToDropShadows=false
      bMiteredCorners=false
      bSeparateFillAlpha=false
      outlineMaterial=null
      outlineSize=0
      outlineColor=<rgba 0 0 0 1>
     }
    }
   }
  }]
}
