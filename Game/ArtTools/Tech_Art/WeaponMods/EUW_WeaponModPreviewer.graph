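# Editor utility widget for previewing weapon mods on a weapon actor spawned near the player pawn in a running PIE world.
# Details views edit test-data class lists, whose default objects' soft references are blocking-loaded into weapon and mod list views on change or refresh.
# Selection respawns the actor if the weapon differs, restores default mods and applies the selected mod.
# Back, Start Cycling and Next buttons have no graph logic.
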EditorUtilityWidgetBlueprint /Game/ArtTools/Tech_Art/WeaponMods/EUW_WeaponModPreviewer
graphs [EventGraph {nodes [
   "On Clicked (Button_Refresh)"() {is=title K2Node_ComponentBoundEvent}
   > OnPropertyChanged_WIDs!(execute, self, PropertyName:Name=None) -> (then)
   > OnPropertyChanged_WMIDs!(execute, self, PropertyName:Name=None) -> (then)
   n1 = OnPropertyChanged_WMIDs!() -> (OutputDelegate, then, PropertyName:Name) {is=title K2Node_CustomEvent}
   > PrintString(InString=Append(A=" property changed ", B="To String (Name)"(InName=@n1.PropertyName)))
   > ClearListItems(self=Get ListView_WMIDs() -> (ListView_WMIDs:Common List View Object Reference))
   > Clear(TargetArray:Array of Object References=Get TotalModList() -> (TotalModList:Array of Object References))
   > n2 = For Each Loop(Array:Array of Test Data Mod List Base Class References=Get ModList() -> (ModList:Array of Test Data Mod List Base Class References)) -> (Array Element:Test Data Mod List Base Class Reference)
   Set List Items!(execute=@n2.Completed, self:List View Object Reference=Get ListView_WMIDs(), InListItems:Array of Object References=Get TotalModList()) -> (then) {is=title K2Node_CallFunction}
   Branch(execute=@n2.LoopBody, Condition=IsValidClass!(Class:Object Class Reference=@n2) -> (ReturnValue:bool))
   > n3 = GetClassDefaultObject!(execute, Class:Object Class Reference=@n2) -> (then, ReturnValue:Object Reference)
   > n4 = Cast To TestData_ModListBase!(execute, Object=@n3) -> (then, CastFailed, AsTest Data Mod List Base:Test Data Mod List Base Object Reference) {is=title K2Node_DynamicCast}
   > Is Valid(InputObject=@n4)
   > n5 = For Each Loop(Array:Array of Fort Weapon Mod Item Definition Soft Object References=Get ModList!(self:Test Data Mod List Base Object Reference=@n4) -> (ModList:Array of Fort Weapon Mod Item Definition Soft Object References)) -> (Array Element:Fort Weapon Mod Item Definition Soft Object Reference)
   > Branch(Condition=IsValidSoftObjectReference(SoftObjectReference=@n5))
   > n6 = LoadAsset_Blocking(Asset=@n5)
   > Is Valid(InputObject=@n6)
   > Add(TargetArray:Array of Object References=Get TotalModList(), NewItem:"Object Reference (by ref)"=@n6) {is=title K2Node_CallArrayFunction}
   n7 = OnPropertyChanged_WIDs!() -> (OutputDelegate, then, PropertyName:Name) {is=title K2Node_CustomEvent}
   > PrintString(InString=Append(A=" property changed ", B="To String (Name)"(InName=@n7.PropertyName)))
   > ClearListItems(self=Get ListView_WIDs() -> (ListView_WIDs:Common List View Object Reference))
   > Clear(TargetArray=Get TotalWeaponList() -> (TotalWeaponList:Array of Object References))
   > n8 = For Each Loop(Array:Array of Test Data Weapon List Base Class References=Get WeaponList() -> (WeaponList:Array of Test Data Weapon List Base Class References)) -> (Array Element:Test Data Weapon List Base Class Reference)
   Set List Items(execute=@n8.Completed, self=Get ListView_WIDs(), InListItems=Get TotalWeaponList())
   Branch(execute=@n8.LoopBody, Condition=IsValidClass(Class=@n8))
   > n9 = GetClassDefaultObject(Class=@n8)
   > n10 = Cast To TestData_WeaponListBase!(execute, Object=@n9) -> (then, CastFailed, AsTest Data Weapon List Base:Test Data Weapon List Base Object Reference) {is=title K2Node_DynamicCast}
   > n11 = For Each Loop(Array:Array of Fort Weapon Item Definition Soft Object References=Get WeaponList!(self:Test Data Weapon List Base Object Reference=@n10) -> (WeaponList:Array of Fort Weapon Item Definition Soft Object References)) -> (Array Element:Fort Weapon Item Definition Soft Object Reference)
   > Branch(Condition=IsValidSoftObjectReference(SoftObjectReference=@n11))
   > n12 = LoadAsset_Blocking(Asset=@n11)
   > Is Valid(InputObject=@n12)
   > Add(TargetArray=Get TotalWeaponList(), NewItem=@n12)
   n13 = OnSelectionChanged_Weapon!() -> (OutputDelegate, then, Item:Object Reference, bIsSelected:bool=~) {is=title K2Node_CustomEvent}
   > Branch(Condition=@n13.bIsSelected)
   > n14 = Cast To FortWeaponItemDefinition!(execute, Object=@n13.Item) -> (then, CastFailed, AsFort Weapon Item Definition:Fort Weapon Item Definition Object Reference) {is=title K2Node_DynamicCast}
   > Set SelectedWeaponWID(SelectedWeaponWID:Fort Weapon Item Definition Object Reference=@n14) -> (Output_Get:Fort Weapon Item Definition Object Reference)
   > TryUpdateWeaponModDisplay()
   n15 = OnSelectionChanged_Mod!() -> (OutputDelegate, then, Item:Object Reference, bIsSelected:bool=~) {is=title K2Node_CustomEvent}
   > Branch(Condition=@n15.bIsSelected)
   > n16 = Cast To FortWeaponModItemDefinition!(execute, Object=@n15.Item) -> (then, CastFailed, AsFort Weapon Mod Item Definition:Fort Weapon Mod Item Definition Object Reference) {is=title K2Node_DynamicCast}
   > Set SelectedModWMID(SelectedModWMID:Fort Weapon Mod Item Definition Object Reference=@n16) -> (Output_Get:Fort Weapon Mod Item Definition Object Reference)
   > TryUpdateWeaponModDisplay()
   TryUpdateWeaponModDisplay() {is=title K2Node_CustomEvent}
   > Get!(execute:Exec) -> (then:Exec, else:Exec, SelectedWeaponWID:Fort Weapon Item Definition Object Reference)
   > Get!(execute) -> (then, else, SelectedModWMID:Fort Weapon Mod Item Definition Object Reference)
   > SpawnPreviewWeaponIfNeeded()
   > n17 = Get!(execute) -> (then, else, PreviewWeaponActor:Fort Weapon Object Reference)
   > RemoveAllPreviewWeaponMods!(execute, PreviewWeapon:Fort Weapon Object Reference=@n17, bRestoreDefaultModSlotsFromWID:bool=true) -> (then)
   > TryAddPreviewWeaponMod!(
    execute
    WeaponModItem:Fort Weapon Mod Item Definition Object Reference=Get SelectedModWMID() -> (SelectedModWMID:Fort Weapon Mod Item Definition Object Reference)
    PreviewWeapon:Fort Weapon Object Reference=@n17
   ) -> (then, ReturnValue:Static Mesh Component Object Reference)
   SpawnPreviewWeaponIfNeeded() {is=title K2Node_CustomEvent}
   > n18 = Is Valid(InputObject=Get PreviewWeaponActor() -> (PreviewWeaponActor:Fort Weapon Object Reference))
   Is Valid(exec=[@n28.then, @n18.Is Not Valid], InputObject=Get SelectedWeaponWID() -> (SelectedWeaponWID:Fort Weapon Item Definition Object Reference))
   > n19 = GetPIEWorlds!(execute, bIncludeDedicatedServer:bool=false) -> (then, ReturnValue:Array of World Object References)
   > Branch(Condition="integer > integer"(A=Length(TargetArray:Array of World Object References=@n19)))
   n20 = "Get (a copy)"(Array:Array of World Object References=@n19) -> (Output:World Object Reference)
   n21 = GetPlayerPawn!(WorldContextObject:Object Reference=@n20, PlayerIndex:int=0) -> (ReturnValue:Pawn Object Reference)
   > Is Valid(InputObject=@n21)
   n22 = BreakTransform(InTransform=Get Actor Transform(self=@n21))
   n23 = float + float(A=BreakRotator(InRot=@n22.Rotation).Yaw)
   > n24 = SpawnWeaponModPreviewActor!(
    execute
    WorldContextObject:Object Reference=@n20
    WeaponItemDef:Fort Weapon Item Definition Object Reference=Get SelectedWeaponWID()
    Transform:Transform=MakeTransform(Location=vector + vector(A=@n22, B="0,0,100"), Rotation=MakeRotator(Yaw=@n23))
   ) -> (then, ReturnValue:Fort Weapon Object Reference)
   > n25 = Set PreviewWeaponActor(PreviewWeaponActor:Fort Weapon Object Reference=@n24) -> (Output_Get:Fort Weapon Object Reference)
   > Is Valid(InputObject=@n25)
   n26 = Get PreviewWeaponActor()
   Branch(
    execute=@n18.Is Valid
    Condition="Not Equal (Object)"(
     A:Object Reference=GetWeaponData!(self:Fort Weapon Object Reference=@n26) -> (ReturnValue:Fort Weapon Item Definition Object Reference)
     B:Object Reference=Get SelectedWeaponWID()
    ) -> (ReturnValue:bool=~) {is=title K2Node_PromotableOperator}
   )
   n27 = Get PreviewWeaponActor()
   > Destroy Actor(self=@n27)
   > n28 = Set Actor Hidden In Game!(execute, self:Actor Object Reference=@n27, bNewHidden:bool=true) -> (then) {is=title K2Node_CallFunction}
   Event Construct()
   > n29 = Sequence()
   > SetObject(
    self=[
     Get DetailsView_ID_WIDS() -> (DetailsView_ID_WIDS:Details View Object Reference)
     Get DetailsView_ID_WMIDS() -> (DetailsView_ID_WMIDS:Details View Object Reference)
    ]
    NewObject=Self-Reference()
   )
   Bind Event to On Property Changed(execute=[@n30.then, @n29.then_1], self:Property View Base Object Reference=Get DetailsView_ID_WIDS(), Delegate=Create Event()) {is=title K2Node_AddDelegate}
   > Bind Event to On Property Changed(self=Get DetailsView_ID_WMIDS(), Delegate=Create Event())
   > Bind Event to On Item Selection Changed(self:List View Object Reference=Get ListView_WIDs(), Delegate=Create Event()) {is=title K2Node_AddDelegate}
   > Bind Event to On Item Selection Changed(self=Get ListView_WMIDs(), Delegate=Create Event())
   n30 = Event OnInitialized()
   Get SelectedWeaponWID()
  ]}]
variables [
 ModList {type="array<Class>", default []}
 PreviewWeaponActor {type=FortWeapon, default=null}
 SelectedModWMID {type=FortWeaponModItemDefinition, default=null}
 SelectedWeaponWID {type=FortWeaponItemDefinition, default=null}
 TotalModList {type="array<Object>", default []}
 TotalWeaponList {type="array<Object>", default []}
 WeaponList {type="array<Class>", default []}
]
widgets {
 info {rootWidgetClass=&VerticalBox, widgetCount=23}
 widgets [
  VerticalBox_39 {widgetClassPath=&VerticalBox, widget=VerticalBox_39}
  HBox_Row1 {widgetClassPath=&HorizontalBox, parent=VerticalBox_39, widget=HBox_Row1, slot {class=VerticalBoxSlot}}
  SizeBox_WeaponImage {
   widgetClassPath=&SizeBox
   parent=HBox_Row1
   widget=SizeBox_WeaponImage
   properties {bOverride_HeightOverride=true, bOverride_WidthOverride=true, heightOverride=512, widthOverride=512}
   slot {class=HorizontalBoxSlot, properties {padding=<ltrb 32 32 32 32>}}
  }
  Image_83 {widgetClassPath=&Image, bIsVariable=true, parent=SizeBox_WeaponImage, widget=Image_83, slot {class=SizeBoxSlot}}
  VBox_TopButtons {
   widgetClassPath=&VerticalBox
   parent=HBox_Row1
   widget=VBox_TopButtons
   slot {class=HorizontalBoxSlot, properties {size {sizeRule=Fill, value=1}}}
  }
  HBox_PlayButtons {
   widgetClassPath=&HorizontalBox
   parent=VBox_TopButtons
   widget=HBox_PlayButtons
   slot {class=VerticalBoxSlot, properties {horizontalAlignment=HAlign_Center, verticalAlignment=VAlign_Top}}
  }
  Button_Back {widgetClassPath=&EditorUtilityButton, parent=HBox_PlayButtons, widget=Button_Back, slot {class=HorizontalBoxSlot}}
  TextBlock_253 {widgetClassPath=&TextBlock, parent=Button_Back, widget=TextBlock_253, properties {text="< Back"}, slot {class=ButtonSlot}}
  Button_Play {widgetClassPath=&EditorUtilityButton, parent=HBox_PlayButtons, widget=Button_Play, slot {class=HorizontalBoxSlot}}
  TextBlock {widgetClassPath=&TextBlock, parent=Button_Play, widget=TextBlock, properties {text=Start Cycling}, slot {class=ButtonSlot}}
  Button_Next {widgetClassPath=&EditorUtilityButton, parent=HBox_PlayButtons, widget=Button_Next, slot {class=HorizontalBoxSlot}}
  TextBlock_1 {widgetClassPath=&TextBlock, parent=Button_Next, widget=TextBlock_1, properties {text="Next >"}, slot {class=ButtonSlot}}
  VBox_Data {
   widgetClassPath=&VerticalBox
   parent=VBox_TopButtons
   widget=VBox_Data
   slot {class=VerticalBoxSlot, properties {padding=<ltrb 0 50 0 0>, size {sizeRule=Fill, value=1}}}
  }
  DetailsView_ID_WIDS {
   widgetClassPath=&DetailsView
   parent=VBox_Data
   widget=DetailsView_ID_WIDS
   properties {categoriesToShow [WID Data], renderTransform {angle=0, scale=<1 1>, shear=<0 0>, translation=<0 -30>}}
   slot {class=VerticalBoxSlot}
  }
  DetailsView_ID_WMIDS {
   widgetClassPath=&DetailsView
   parent=VBox_Data
   widget=DetailsView_ID_WMIDS
   properties {categoriesToShow [Mod Data], renderTransform {angle=0, scale=<1 1>, shear=<0 0>, translation=<0 -30>}}
   slot {class=VerticalBoxSlot}
  }
  SizeBox_0 {
   widgetClassPath=&SizeBox
   parent=VBox_Data
   widget=SizeBox_0
   properties {bOverride_HeightOverride=true, heightOverride=400, maxDesiredHeight=250}
   slot {class=VerticalBoxSlot}
  }
  HBox_Lists {widgetClassPath=&HorizontalBox, parent=SizeBox_0, widget=HBox_Lists, slot {class=SizeBoxSlot}}
  EditorUtilityScrollBox_0 {widgetClassPath=&EditorUtilityScrollBox, parent=HBox_Lists, widget=EditorUtilityScrollBox_0, slot {class=HorizontalBoxSlot}}
  ListView_WIDs {
   widgetClassPath=&CommonListView
   bIsVariable=true
   parent=EditorUtilityScrollBox_0
   widget=ListView_WIDs
   properties {
    entryWidgetClass=&/EW_ModToolItemDefPreviewListEntry.EW_ModToolItemDefPreviewListEntry_C
    scrollBarStyle {
     thickness=8
     draggedThumbImage {
      bIsDynamicallyLoaded=false
      drawAs=RoundedBox
      imageType=NoImage
      mirroring=NoMirror
      resourceName=None
      resourceObject=null
      tiling=NoTile
      imageSize=<0 0>
      margin=<ltrb 0 0 0 0>
      outlineSettings {
       bUseBrushTransparency=false
       roundingType=FixedRadius
       width=0
       color {colorUseRule=UseColor_Specified, specifiedColor=<rgba 0 0 0 0>}
       cornerRadii=<4 4 4 4>
      }
      tintColor {colorUseRule=UseColor_Specified, specifiedColor=<rgba 0.2159 0.2159 0.2159 1>}
      uVRegion {bIsValid=false, max=<0 0>, min=<0 0>}
     }
     horizontalBackgroundImage {
      bIsDynamicallyLoaded=false
      drawAs=NoDrawType
      imageType=NoImage
      mirroring=NoMirror
      resourceName=None
      resourceObject=null
      tiling=NoTile
      imageSize=<0 0>
      margin=<ltrb 0 0 0 0>
      outlineSettings {
       bUseBrushTransparency=false
       roundingType=HalfHeightRadius
       width=0
       color {colorUseRule=UseColor_Specified, specifiedColor=<rgba 0 0 0 0>}
       cornerRadii=<0 0 0 0>
      }
      tintColor {colorUseRule=UseColor_Specified, specifiedColor=<rgba 1 1 1 1>}
      uVRegion {bIsValid=false, max=<0 0>, min=<0 0>}
     }
     horizontalBottomSlotImage {
      bIsDynamicallyLoaded=false
      drawAs=NoDrawType
      imageType=NoImage
      mirroring=NoMirror
      resourceName=None
      resourceObject=null
      tiling=NoTile
      imageSize=<0 0>
      margin=<ltrb 0 0 0 0>
      outlineSettings {
       bUseBrushTransparency=false
       roundingType=HalfHeightRadius
       width=0
       color {colorUseRule=UseColor_Specified, specifiedColor=<rgba 0 0 0 0>}
       cornerRadii=<0 0 0 0>
      }
      tintColor {colorUseRule=UseColor_Specified, specifiedColor=<rgba 1 1 1 1>}
      uVRegion {bIsValid=false, max=<0 0>, min=<0 0>}
     }
     horizontalTopSlotImage {
      bIsDynamicallyLoaded=false
      drawAs=NoDrawType
      imageType=NoImage
      mirroring=NoMirror
      resourceName=None
      resourceObject=null
      tiling=NoTile
      imageSize=<0 0>
      margin=<ltrb 0 0 0 0>
      outlineSettings {
       bUseBrushTransparency=false
       roundingType=HalfHeightRadius
       width=0
       color {colorUseRule=UseColor_Specified, specifiedColor=<rgba 0 0 0 0>}
       cornerRadii=<0 0 0 0>
      }
      tintColor {colorUseRule=UseColor_Specified, specifiedColor=<rgba 1 1 1 1>}
      uVRegion {bIsValid=false, max=<0 0>, min=<0 0>}
     }
     hoveredThumbImage {
      bIsDynamicallyLoaded=false
      drawAs=RoundedBox
      imageType=NoImage
      mirroring=NoMirror
      resourceName=None
      resourceObject=null
      tiling=NoTile
      imageSize=<0 0>
      margin=<ltrb 0 0 0 0>
      outlineSettings {
       bUseBrushTransparency=false
       roundingType=FixedRadius
       width=0
       color {colorUseRule=UseColor_Specified, specifiedColor=<rgba 0 0 0 0>}
       cornerRadii=<4 4 4 4>
      }
      tintColor {colorUseRule=UseColor_Specified, specifiedColor=<rgba 0.2159 0.2159 0.2159 1>}
      uVRegion {bIsValid=false, max=<0 0>, min=<0 0>}
     }
     normalThumbImage {
      bIsDynamicallyLoaded=false
      drawAs=RoundedBox
      imageType=NoImage
      mirroring=NoMirror
      resourceName=None
      resourceObject=null
      tiling=NoTile
      imageSize=<0 0>
      margin=<ltrb 0 0 0 0>
      outlineSettings {
       bUseBrushTransparency=false
       roundingType=FixedRadius
       width=0
       color {colorUseRule=UseColor_Specified, specifiedColor=<rgba 0 0 0 0>}
       cornerRadii=<4 4 4 4>
      }
      tintColor {colorUseRule=UseColor_Specified, specifiedColor=<rgba 0.0953 0.0953 0.0953 1>}
      uVRegion {bIsValid=false, max=<0 0>, min=<0 0>}
     }
     verticalBackgroundImage {
      bIsDynamicallyLoaded=false
      drawAs=NoDrawType
      imageType=NoImage
      mirroring=NoMirror
      resourceName=None
      resourceObject=null
      tiling=NoTile
      imageSize=<0 0>
      margin=<ltrb 0 0 0 0>
      outlineSettings {
       bUseBrushTransparency=false
       roundingType=HalfHeightRadius
       width=0
       color {colorUseRule=UseColor_Specified, specifiedColor=<rgba 0 0 0 0>}
       cornerRadii=<0 0 0 0>
      }
      tintColor {colorUseRule=UseColor_Specified, specifiedColor=<rgba 1 1 1 1>}
      uVRegion {bIsValid=false, max=<0 0>, min=<0 0>}
     }
     verticalBottomSlotImage {
      bIsDynamicallyLoaded=false
      drawAs=NoDrawType
      imageType=NoImage
      mirroring=NoMirror
      resourceName=None
      resourceObject=null
      tiling=NoTile
      imageSize=<0 0>
      margin=<ltrb 0 0 0 0>
      outlineSettings {
       bUseBrushTransparency=false
       roundingType=HalfHeightRadius
       width=0
       color {colorUseRule=UseColor_Specified, specifiedColor=<rgba 0 0 0 0>}
       cornerRadii=<0 0 0 0>
      }
      tintColor {colorUseRule=UseColor_Specified, specifiedColor=<rgba 1 1 1 1>}
      uVRegion {bIsValid=false, max=<0 0>, min=<0 0>}
     }
     verticalTopSlotImage {
      bIsDynamicallyLoaded=false
      drawAs=NoDrawType
      imageType=NoImage
      mirroring=NoMirror
      resourceName=None
      resourceObject=null
      tiling=NoTile
      imageSize=<0 0>
      margin=<ltrb 0 0 0 0>
      outlineSettings {
       bUseBrushTransparency=false
       roundingType=HalfHeightRadius
       width=0
       color {colorUseRule=UseColor_Specified, specifiedColor=<rgba 0 0 0 0>}
       cornerRadii=<0 0 0 0>
      }
      tintColor {colorUseRule=UseColor_Specified, specifiedColor=<rgba 1 1 1 1>}
      uVRegion {bIsValid=false, max=<0 0>, min=<0 0>}
     }
    }
   }
   slot {class=ScrollBoxSlot, properties {size {sizeRule=Fill, value=1}}}
  }
  EditorUtilityScrollBox_1 {widgetClassPath=&EditorUtilityScrollBox, parent=HBox_Lists, widget=EditorUtilityScrollBox_1, slot {class=HorizontalBoxSlot}}
  ListView_WMIDs {
   widgetClassPath=&CommonListView
   bIsVariable=true
   parent=EditorUtilityScrollBox_1
   widget=ListView_WMIDs
   properties {
    entryWidgetClass=&EW_ModToolItemDefPreviewListEntry_C
    scrollBarStyle {
     thickness=8
     draggedThumbImage {
      bIsDynamicallyLoaded=false
      drawAs=RoundedBox
      imageType=NoImage
      mirroring=NoMirror
      resourceName=None
      resourceObject=null
      tiling=NoTile
      imageSize=<0 0>
      margin=<ltrb 0 0 0 0>
      outlineSettings {
       bUseBrushTransparency=false
       roundingType=FixedRadius
       width=0
       color {colorUseRule=UseColor_Specified, specifiedColor=<rgba 0 0 0 0>}
       cornerRadii=<4 4 4 4>
      }
      tintColor {colorUseRule=UseColor_Specified, specifiedColor=<rgba 0.2159 0.2159 0.2159 1>}
      uVRegion {bIsValid=false, max=<0 0>, min=<0 0>}
     }
     horizontalBackgroundImage {
      bIsDynamicallyLoaded=false
      drawAs=NoDrawType
      imageType=NoImage
      mirroring=NoMirror
      resourceName=None
      resourceObject=null
      tiling=NoTile
      imageSize=<0 0>
      margin=<ltrb 0 0 0 0>
      outlineSettings {
       bUseBrushTransparency=false
       roundingType=HalfHeightRadius
       width=0
       color {colorUseRule=UseColor_Specified, specifiedColor=<rgba 0 0 0 0>}
       cornerRadii=<0 0 0 0>
      }
      tintColor {colorUseRule=UseColor_Specified, specifiedColor=<rgba 1 1 1 1>}
      uVRegion {bIsValid=false, max=<0 0>, min=<0 0>}
     }
     horizontalBottomSlotImage {
      bIsDynamicallyLoaded=false
      drawAs=NoDrawType
      imageType=NoImage
      mirroring=NoMirror
      resourceName=None
      resourceObject=null
      tiling=NoTile
      imageSize=<0 0>
      margin=<ltrb 0 0 0 0>
      outlineSettings {
       bUseBrushTransparency=false
       roundingType=HalfHeightRadius
       width=0
       color {colorUseRule=UseColor_Specified, specifiedColor=<rgba 0 0 0 0>}
       cornerRadii=<0 0 0 0>
      }
      tintColor {colorUseRule=UseColor_Specified, specifiedColor=<rgba 1 1 1 1>}
      uVRegion {bIsValid=false, max=<0 0>, min=<0 0>}
     }
     horizontalTopSlotImage {
      bIsDynamicallyLoaded=false
      drawAs=NoDrawType
      imageType=NoImage
      mirroring=NoMirror
      resourceName=None
      resourceObject=null
      tiling=NoTile
      imageSize=<0 0>
      margin=<ltrb 0 0 0 0>
      outlineSettings {
       bUseBrushTransparency=false
       roundingType=HalfHeightRadius
       width=0
       color {colorUseRule=UseColor_Specified, specifiedColor=<rgba 0 0 0 0>}
       cornerRadii=<0 0 0 0>
      }
      tintColor {colorUseRule=UseColor_Specified, specifiedColor=<rgba 1 1 1 1>}
      uVRegion {bIsValid=false, max=<0 0>, min=<0 0>}
     }
     hoveredThumbImage {
      bIsDynamicallyLoaded=false
      drawAs=RoundedBox
      imageType=NoImage
      mirroring=NoMirror
      resourceName=None
      resourceObject=null
      tiling=NoTile
      imageSize=<0 0>
      margin=<ltrb 0 0 0 0>
      outlineSettings {
       bUseBrushTransparency=false
       roundingType=FixedRadius
       width=0
       color {colorUseRule=UseColor_Specified, specifiedColor=<rgba 0 0 0 0>}
       cornerRadii=<4 4 4 4>
      }
      tintColor {colorUseRule=UseColor_Specified, specifiedColor=<rgba 0.2159 0.2159 0.2159 1>}
      uVRegion {bIsValid=false, max=<0 0>, min=<0 0>}
     }
     normalThumbImage {
      bIsDynamicallyLoaded=false
      drawAs=RoundedBox
      imageType=NoImage
      mirroring=NoMirror
      resourceName=None
      resourceObject=null
      tiling=NoTile
      imageSize=<0 0>
      margin=<ltrb 0 0 0 0>
      outlineSettings {
       bUseBrushTransparency=false
       roundingType=FixedRadius
       width=0
       color {colorUseRule=UseColor_Specified, specifiedColor=<rgba 0 0 0 0>}
       cornerRadii=<4 4 4 4>
      }
      tintColor {colorUseRule=UseColor_Specified, specifiedColor=<rgba 0.0953 0.0953 0.0953 1>}
      uVRegion {bIsValid=false, max=<0 0>, min=<0 0>}
     }
     verticalBackgroundImage {
      bIsDynamicallyLoaded=false
      drawAs=NoDrawType
      imageType=NoImage
      mirroring=NoMirror
      resourceName=None
      resourceObject=null
      tiling=NoTile
      imageSize=<0 0>
      margin=<ltrb 0 0 0 0>
      outlineSettings {
       bUseBrushTransparency=false
       roundingType=HalfHeightRadius
       width=0
       color {colorUseRule=UseColor_Specified, specifiedColor=<rgba 0 0 0 0>}
       cornerRadii=<0 0 0 0>
      }
      tintColor {colorUseRule=UseColor_Specified, specifiedColor=<rgba 1 1 1 1>}
      uVRegion {bIsValid=false, max=<0 0>, min=<0 0>}
     }
     verticalBottomSlotImage {
      bIsDynamicallyLoaded=false
      drawAs=NoDrawType
      imageType=NoImage
      mirroring=NoMirror
      resourceName=None
      resourceObject=null
      tiling=NoTile
      imageSize=<0 0>
      margin=<ltrb 0 0 0 0>
      outlineSettings {
       bUseBrushTransparency=false
       roundingType=HalfHeightRadius
       width=0
       color {colorUseRule=UseColor_Specified, specifiedColor=<rgba 0 0 0 0>}
       cornerRadii=<0 0 0 0>
      }
      tintColor {colorUseRule=UseColor_Specified, specifiedColor=<rgba 1 1 1 1>}
      uVRegion {bIsValid=false, max=<0 0>, min=<0 0>}
     }
     verticalTopSlotImage {
      bIsDynamicallyLoaded=false
      drawAs=NoDrawType
      imageType=NoImage
      mirroring=NoMirror
      resourceName=None
      resourceObject=null
      tiling=NoTile
      imageSize=<0 0>
      margin=<ltrb 0 0 0 0>
      outlineSettings {
       bUseBrushTransparency=false
       roundingType=HalfHeightRadius
       width=0
       color {colorUseRule=UseColor_Specified, specifiedColor=<rgba 0 0 0 0>}
       cornerRadii=<0 0 0 0>
      }
      tintColor {colorUseRule=UseColor_Specified, specifiedColor=<rgba 1 1 1 1>}
      uVRegion {bIsValid=false, max=<0 0>, min=<0 0>}
     }
    }
   }
   slot {class=ScrollBoxSlot, properties {size {sizeRule=Fill, value=1}}}
  }
  Button_Refresh {widgetClassPath=&EditorUtilityButton, parent=VBox_TopButtons, widget=Button_Refresh, slot {class=VerticalBoxSlot}}
  TextBlock_2 {widgetClassPath=&TextBlock, parent=Button_Refresh, widget=TextBlock_2, properties {text=Refresh Lists}, slot {class=ButtonSlot}}
 ]
}
